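# Landscape layer function that adds the Scree material to the Hermes terrain, gated by the BiomeB layer switch.
# World-UV scree textures build desaturated color, normals, height and displacement, which MF_Hermes_Scree_BiomeColor then adjusts using terrain masks.
# The result height-blends onto the input material with contrast, using diffuse alpha as height.
# A separate mobile layer has its own blend, and extra outputs expose the layer mask and a static-mesh material with mobile-derived normals.
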MaterialFunction /Game/Hermes/Terrain/Landscape/Materials/Layers/MF_Hermes_Scree
graph {nodes [
  n1 = NamedRerouteUsage()
  n2 = NamedRerouteUsage()
  n3 = NamedRerouteUsage()
  n4 = NamedRerouteUsage()
  n5 = NamedRerouteUsage()
  n6 = MaterialFunctionCall!(InputMaterial=@n1, LayerMaterial=@n2, LayerWeight=@n3, UseContrast=@n4, Contrast=@n5) {materialFunction=/Game/Athena/Helios/Environments/Landscape/Materials/MaterialFunctions_Utilities/MF_HeightBlendLandscapeLayers_OpacityMaskDisplaceCustom}
  FunctionOutput(None=LandscapeLayerSwitch!(LayerUsed=@n6.BlendedMaterial, LayerNotUsed=@n1) {parameterName=BiomeB}) {outputName=BlendedMaterial, sortPriority=-300}
  FunctionOutput(None=ShadingPathSwitch(Default=NamedRerouteUsage(), Mobile=NamedRerouteUsage())) {outputName=LayerMask, sortPriority=-1}
  n7 = NamedRerouteUsage()
  n8 = MaterialFunctionCall!(
   InputMaterial=@n7
   LayerMaterial=NamedRerouteUsage()
   LayerWeight=NamedRerouteUsage()
   UseContrast=NamedRerouteUsage()
   Contrast=NamedRerouteUsage()
  ) {materialFunction=/Game/Athena/Helios/Environments/Landscape/Materials/MaterialFunctions_Utilities/MF_HeightBlendLandscapeLayers_OpacityMask}
  FunctionOutput(None=LandscapeLayerSwitch(LayerUsed=@n8.BlendedMaterial, LayerNotUsed=@n7) {parameterName=BiomeB}) {outputName=MobileBlendedMaterial, sortPriority=-100}
  n9 = NamedRerouteUsage()
  n10 = StaticBool()
  FunctionOutput(None=ShadingPathSwitch(
    Default=NamedRerouteUsage()
    Mobile=MaterialFunctionCall!(MobileMaterial=@n9, TransformTangents=@n10) {materialFunction=/Game/Athena/Helios/Environments/Landscape/Materials/MaterialFunctions_Utilities/MF_MobileDeriveNormals}.ResultMaterial
   )) {outputName=StaticMeshMaterial, sortPriority=10}
  n11 = MaterialFunctionCall!(Scale=ScalarParameter() {defaultValue=900, group=Scree, parameterName=Scree_WorldScale}) {materialFunction=/Game/Athena/Apollo/Environments/Landscape/Materials/Biome/CommonMaterials/MF_SimpleWorldUV}
  n12 = TextureSampleParameter2D(UVs=@n11.Result) {
   group=Scree
   parameterName=Scree_S
   samplerSource=SSM_Wrap_WorldGroupSettings
   samplerType=SAMPLERTYPE_Masks
   texture=/Game/Environments/Asteria/Terrain/Textures/T_Asteria_Scree_01_S
  }
  LinearInterpolate(
   A=ScalarParameter() {defaultValue=0.5, group=Scree, parameterName=Scree_SpecFlat}
   B=@n12.R
   Alpha=ScalarParameter() {defaultValue=1, group=Scree, parameterName=Scree_SpecType}
  )
  MaterialFunctionCall(InputMaterial=@n1, LayerMaterial=@n2, LayerWeight=@n3, UseContrast=@n4, Contrast=@n5) {materialFunction=&MF_HeightBlendLandscapeLayers_OpacityMask}
  n13 = FunctionInput(Preview=StaticBool()) {
   bUsePreviewValueAsDefault=true
   description="Default = False"
   inputName=UseTangentNormals
   inputType=FunctionInput_StaticBool
   sortPriority=15
  }
  NamedRerouteDeclaration(None=@n13) {name=UseTangentNormalsBool, nodeColor=<rgba 0.5 0 1 1>}
  NamedRerouteDeclaration(None=FunctionInput(Preview=Constant()) {bUsePreviewValueAsDefault=true, inputName=VariationMasks Mobile, inputType=FunctionInput_Vector4, sortPriority=40}) {name=VariationMasks Mobile, nodeColor=<rgba 1 0.8498 0.0704 1>}
  n14 = TextureSampleParameter2D(UVs=@n11.Result) {
   group=Scree
   parameterName=Scree_D
   samplerSource=SSM_Wrap_WorldGroupSettings
   texture=/Game/Environments/Asteria/Terrain/Textures/T_Asteria_Scree_01_D
  }
  n15 = TextureSampleParameter2D(UVs=@n11.Result) {
   group=Scree
   parameterName=Scree_N
   samplerSource=SSM_Wrap_WorldGroupSettings
   samplerType=SAMPLERTYPE_Normal
   texture=/Game/Environments/Asteria/Terrain/Textures/T_Asteria_Scree_01_N
  }
  n16 = SetMaterialAttributes!(
   MaterialAttributes
   BaseColor=Desaturation(None=@n14.RGB, Fraction=ScalarParameter() {defaultValue=0.5, group=Scree, parameterName=Scree_DesaturateAmt})
   Specular=@n12.R
   Roughness=@n12.R
   Normal=MaterialFunctionCall!(Normals=@n15.RGB, TransformToTangentSpace=NamedRerouteUsage(), Normal Strength) {materialFunction=/Game/Athena/Apollo/Environments/Landscape/Materials/Biome/CommonMaterials/MF_HQ_WorldAlignedNormal}.Result
   OpacityMask=NamedRerouteDeclaration(None=@n14.A) {name=ScreeHeight, nodeColor=<rgba 0.5 0 1 1>}
   Displacement=LinearInterpolate(B=@n14.A, Alpha=ScalarParameter() {defaultValue=0.8, group=Displacement, parameterName=Scree_DisplacementAmt}) {constA=0.5}
  ) {attributeSetTypes [
    #1
    #2
    #3
    #4
    #5
    #6
   ]}
  NamedRerouteDeclaration(None=MaterialFunctionCall!(
    Material=@n16
    TerrainMasks_QuadA=NamedRerouteUsage()
    TerrainMasks_QuadB=NamedRerouteUsage()
    TerrainMasks_QuadC=NamedRerouteUsage()
    TerrainMasks_QuadD=NamedRerouteUsage()
    GrassNoise
   ) {materialFunction=/Game/Hermes/Terrain/Landscape/Materials/Layers/BiomeColor/MF_Hermes_Scree_BiomeColor}.Result) {name=LayerMaterial, nodeColor=<rgba 0 1 0 1>}
  NamedRerouteDeclaration(None=FunctionInput(Preview=Constant()) {bUsePreviewValueAsDefault=true, inputName=LayerWeight, inputType=FunctionInput_Scalar, sortPriority=15}) {name=LayerWeight, nodeColor=<rgba 0.5 0 1 1>}
  NamedRerouteDeclaration(None=@n6.LayerMask) {name=LayerMask, nodeColor=<rgba 0.7344 0 1 1>}
  n17 = SetMaterialAttributes!(MaterialAttributes)
  NamedRerouteDeclaration(None=FunctionInput(Preview=@n17) {bUsePreviewValueAsDefault=true, inputName=Material, inputType=FunctionInput_MaterialAttributes}) {name=InputMaterial, nodeColor=<rgba 0 1 0 1>}
  NamedRerouteDeclaration(None=FunctionInput(Preview=Constant()) {bUsePreviewValueAsDefault=true, inputName=TerrainMasks_QuadA, inputType=FunctionInput_Vector4, sortPriority=20}) {name=TerrainMasks_QuadA, nodeColor=<rgba 0.5 0 1 1>}
  NamedRerouteDeclaration(None=FunctionInput(Preview=Constant()) {bUsePreviewValueAsDefault=true, inputName=TerrainMasks_QuadC, inputType=FunctionInput_Vector4, sortPriority=30}) {name=TerrainMasks_QuadC, nodeColor=<rgba 0.5 0 1 1>}
  NamedRerouteDeclaration(None=@n8.LayerMask) {name=MobileLayerMask, nodeColor=<rgba 0 0.9062 1 1>}
  NamedRerouteDeclaration(None=FunctionInput(Preview=Constant()) {bUsePreviewValueAsDefault=true, inputName=TerrainMasks_QuadD, inputType=FunctionInput_Vector4, sortPriority=35}) {name=TerrainMasks_QuadD, nodeColor=<rgba 0.5 0 1 1>}
  NamedRerouteDeclaration(None=FunctionInput(Preview=Constant()) {bUsePreviewValueAsDefault=true, inputName=TerrainMasks_QuadB, inputType=FunctionInput_Vector4, sortPriority=25}) {name=TerrainMasks_QuadB, nodeColor=<rgba 0.5 0 1 1>}
  NamedRerouteDeclaration(None=FunctionInput(Preview=@n17) {bUsePreviewValueAsDefault=true, inputName=MobileMaterial, inputType=FunctionInput_MaterialAttributes, sortPriority=10}) {name=MobileInputMaterial, nodeColor=<rgba 0 0.5 1 1>}
  NamedRerouteDeclaration(None=MaterialFunctionCall!(
    TerrainMasks_QuadA=NamedRerouteUsage()
    TerrainMasks_QuadB=NamedRerouteUsage()
    TerrainMasks_QuadC=NamedRerouteUsage()
    TerrainMasks_QuadD=NamedRerouteUsage()
   ) {materialFunction=/Game/Hermes/Terrain/Landscape/Materials/Layers/QualityLevel/MF_Hermes_Scree_Mobile}.Result) {name=MobileLayerMaterial, nodeColor=<rgba 0 0.5 1 1>}
  NamedRerouteDeclaration(None=StaticBool() {value=true}) {name=UseContrast, nodeColor=<rgba 0.5 0 1 1>}
  NamedRerouteDeclaration(None=ScalarParameter() {defaultValue=0.5, group=Scree, parameterName=ScreeLayerBlendContrast}) {name=LayerBlendContrast, nodeColor=<rgba 0.5 0 1 1>}
  Saturate(None=Multiply(A=VectorParameter() {group=ForestFloor, parameterName=Scree_ColorTint, defaultValue=<rgba 1 1 1 0>}.RGB, B=@n14.RGB))
  TextureSample() {samplerSource=SSM_Wrap_WorldGroupSettings, texture=&T_Asteria_Scree_01_D}
  TextureSample() {
   samplerSource=SSM_Wrap_WorldGroupSettings
   samplerType=SAMPLERTYPE_Masks
   texture=&T_Asteria_Scree_01_S
  }
  TextureSample() {
   samplerSource=SSM_Wrap_WorldGroupSettings
   samplerType=SAMPLERTYPE_Normal
   texture=&T_Asteria_Scree_01_N
  }
 ]}
